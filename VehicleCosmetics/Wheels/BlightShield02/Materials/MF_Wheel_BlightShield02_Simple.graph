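# Simplified wheel material function for the BlightShield02 cosmetic, overriding base color and emissive from MF_WheelParent_Simple.
# Two tilings of a noise heightmap brighten the parent color, blended by vertex red, and a Fresnel term tints edges toward FresnelColor.
# View-dependent glints multiply squared heightmap noise with packed noise sampled by both reflection-plus-normal and UVs.
# The rim texture's blue channel and vertex red control glint strength in emissive, with the Fresnel tint on top.
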MaterialFunction /VehicleCosmetics/Wheels/BlightShield02/Materials/MF_Wheel_BlightShield02_Simple
graph {nodes [
  n1 = MaterialFunctionCall() {materialFunction=&MF_WheelParent_Simple}
  n2 = GetMaterialAttributes(None=@n1.MaterialAttributes) {attributeGetTypes [#1, #2]}
  n3 = TextureSample(UVs=Add(A=TextureCoordinate() {uTiling=3, vTiling=3}, B=Constant2Vector() {g=0.2})) {texture=/VehicleCosmetics/Wheels/BlightShield02/Textures/Universal_Noise_Heightmap_14}
  n4 = Multiply(A=@n3.R, B=@n2.BaseColor)
  n5 = Constant3Vector() {constant=<rgba 0 0 1 0>}
  n6 = OneMinus(None=DotProduct(A=Transform(None=@n5), B=CameraVectorWS()))
  n7 = TextureSample(UVs=Add(A=TextureCoordinate() {uTiling=2}, B=Constant2Vector() {g=0.2})) {texture=&Universal_Noise_Heightmap_14}
  n8 = Multiply(A=@n7.R, B=@n2.BaseColor)
  n9 = LinearInterpolate(
   A=LinearInterpolate(A=@n2.BaseColor, B=Add(A=@n4, B=@n2.BaseColor), Alpha=ConstantBiasScale(None=@n6) {bias=-0.2, scale=1})
   B=LinearInterpolate(A=@n2.BaseColor, B=Add(A=@n8, B=@n2.BaseColor), Alpha=ConstantBiasScale(None=@n6) {bias=-0.2, scale=3.5})
   Alpha=VertexColor().R
  )
  n10 = Saturate(None=ConstantBiasScale(None=@n6) {bias=0.1, scale=1.5})
  n11 = Desaturation(
   None=LinearInterpolate(A=@n9, B=VectorParameter() {group=4_Surface Defaults, parameterName=FresnelColor, defaultValue=<rgba 0.114 0.216 0.389 1>}.RGB, Alpha=@n10)
   Fraction=Constant() {r=0.1}
  )
  n12 = Add(A=ReflectionVectorWS(), B=@n2.Normal)
  n13 = TextureSample(UVs=Multiply(A=ComponentMask(None=@n12) {g=true, r=true}) {constB=0.0625}) {texture=/VehicleCosmetics/Wheels/BlightShield02/Textures/Noise_Generic_01_Pack}
  n14 = TextureSample(UVs=TextureCoordinate() {uTiling=3, vTiling=3}) {texture=&Noise_Generic_01_Pack}
  n15 = Multiply(A=Multiply(A=@n3.R, B=@n3.R), B=Multiply(A=Multiply(A=@n13.B, B=@n14.B)) {constB=4096})
  n16 = TextureSample(UVs=TextureCoordinate() {uTiling=3, vTiling=3}) {texture=&Noise_Generic_01_Pack}
  n17 = Multiply(A=Multiply(A=@n7.R, B=@n7.R), B=Multiply(A=Multiply(A=@n13.B, B=@n16.B)) {constB=4096})
  n18 = VertexColor()
  n19 = Multiply(A=@n9, B=Add(A=@n2.BaseColor, B=LinearInterpolate(A=@n15, B=@n17, Alpha=@n18.R)))
  n20 = Multiply(A=@n19) {constB=0.3}
  n21 = TextureSample() {
   samplerSource=SSM_Wrap_WorldGroupSettings
   samplerType=SAMPLERTYPE_LinearColor
   texture=/VehicleCosmetics/Wheels/BlightShield02/Textures/T_Rim_BlightShield02_RGB
  }
  n22 = LinearInterpolate(A=LinearInterpolate(A=LinearInterpolate(A=@n20, B=@n19, Alpha=@n21.B), B=@n19, Alpha=VertexColor().R), B=@n11, Alpha=@n10)
  FunctionOutput(None=SetMaterialAttributes!(MaterialAttributes=@n1.MaterialAttributes, BaseColor=Multiply(A=@n11) {constB=0.9}, EmissiveColor=Multiply(A=@n22) {constB=1.5}) {attributeSetTypes [#1, #3]})
 ]}
